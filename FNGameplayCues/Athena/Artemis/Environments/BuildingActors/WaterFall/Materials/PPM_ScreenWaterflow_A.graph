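# Post-process material that distorts the screen with flowing water, apparently for the Artemis waterfall building actor.
# Two chained panning noise layers warp aspect-corrected UVs that sample a plasma normal and matching mask.
# A curve-atlas dissolve, bent across the screen by a sine and driven by the Time parameter, cuts that pattern, while refraction strength fades as Time rises.
# The masked offset shifts the scene-color lookup; mobile passes the scene through undistorted.
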Material /FNGameplayCues/Athena/Artemis/Environments/BuildingActors/WaterFall/Materials/PPM_ScreenWaterflow_A
graph {
 properties {materialDomain=MD_PostProcess, pixelDepthOffsetMode=PDOM_Legacy, usedShadingModels=Default Lit}
 nodes [
  n1 = ScreenPosition()
  n2 = MaterialFunctionCall(Float2=Constant2Vector() {g=9, r=21}) {materialFunction=&BreakOutFloat2Components}
  n3 = MaterialFunctionCall(Float2=MaterialFunctionCall() {materialFunction=&ScreenResolution}.Visible Resolution) {materialFunction=&BreakOutFloat2Components}
  n4 = Divide(A=Divide(A=@n2.R, B=@n2.R), B=Divide(A=@n3.R, B=@n3.R))
  n5 = ScalarParameter() {defaultValue=1, parameterName=TotalUScale}
  n6 = MaterialFunctionCall(UVs=@n1.ViewportUV, Texture Scale=AppendVector(A=Multiply(A=@n4, B=@n5), B=Constant() {r=1})) {materialFunction=&ScaleUVsByCenter}
  n7 = VectorParameter() {parameterName=Water_SP, defaultValue=<rgba 1 1 0 1>}
  n8 = AppendVector(A=@n7.R, B=@n7.R)
  n9 = Time() {bOverride_Period=true, period=300}
  n10 = Constant()
  n11 = VectorParameter() {parameterName=NoiseA_SP, defaultValue=<rgba 1 1 0 1>}
  n12 = Multiply(A=@n6.UVs, B=AppendVector(A=@n11.R, B=@n11.R))
  n13 = AppendVector(A=@n11.B, B=@n11.B)
  n14 = VectorParameter() {parameterName=NoiseB_SP, defaultValue=<rgba 1 1 0 1>}
  n15 = Multiply(A=@n6.UVs, B=AppendVector(A=@n14.R, B=@n14.R))
  n16 = AppendVector(A=@n14.B, B=@n14.B)
  n17 = TextureObjectParameter() {
   parameterName=NoiseB
   samplerType=SAMPLERTYPE_Normal
   texture=/Game/Athena/Artemis/Environments/BuildingActors/WaterFall/Materials/Texture/T_Lowres_Simple_Perlin_N
  }
  n18 = AppendVector(A=Constant(), B=TextureSample(UVs=Panner(Coordinate=@n15, Time=@n9, Speed=@n16) {speedY=-0.2}, Tex=@n17) {samplerType=SAMPLERTYPE_Normal}.G)
  n19 = ScalarParameter() {parameterName=NoiseB_S}
  n20 = TextureObjectParameter() {
   parameterName=NoiseA
   samplerType=SAMPLERTYPE_LinearGrayscale
   texture=/Game/Gadgets/Assets/VinderTech_GliderChute/Glider_Thrive_Female/FX/Texture/T_CloudBubble
  }
  n21 = TextureSample(UVs=Add(A=Panner(Coordinate=@n12, Time=@n9, Speed=@n13) {speedY=-0.2}, B=Multiply(A=@n18, B=@n19) {constB=0.02}), Tex=@n20) {samplerType=SAMPLERTYPE_Normal}
  n22 = Add(
   A=Panner(Coordinate=Multiply(A=@n6.UVs, B=@n8), Time=@n9, Speed=AppendVector(A=@n7.B, B=@n7.B)) {speedY=-0.1}
   B=Multiply(A=AppendVector(A=@n10, B=@n21.R), B=ScalarParameter() {parameterName=NoiseA_S}) {constB=0.2}
  )
  n23 = TextureObjectParameter() {
   parameterName=Water
   samplerType=SAMPLERTYPE_Normal
   texture=/FNGameplayCues/Athena/Artemis/Environments/BuildingActors/WaterFall/Materials/Texture/T_Plasma_N
  }
  n24 = TextureSample(UVs=@n22, Tex=@n23) {samplerType=SAMPLERTYPE_Normal}
  n25 = AppendVector(A=@n24.R, B=@n24.R)
  n26 = ScalarParameter() {defaultValue=0.2, parameterName=Time}
  n27 = MaterialFunctionCall(Float2=ScreenPosition().ViewportUV) {materialFunction=&BreakOutFloat2Components}
  n28 = Multiply(A=@n27.R) {constB=0.5}
  n29 = Multiply(A=OneMinus(None=@n26), B=Multiply(A=Sine(None=@n28), B=ScalarParameter() {parameterName=CenterBend}) {constB=-0.3})
  n30 = ScalarParameter() {parameterName=DissolveWidth}
  n31 = CurveAtlasRowParameter(CurveTime=Add(A=Add(A=@n29, B=@n26), B=Multiply(A=@n27.G, B=@n30))) {
   atlas=/Game/Athena/Artemis/Environments/BuildingActors/WaterFall/Materials/Texture/CurveAtlas/CLA_WaterFall
   curve=/Game/Athena/Artemis/Environments/BuildingActors/WaterFall/Materials/Texture/CurveAtlas/CL_Post
   defaultValue=20
   parameterName=DissolveCurve
  }
  n32 = Subtract(A=Abs(None=@n25), B=@n31.R)
  n33 = Power(Base=OneMinus(None=Saturate(None=@n32))) {constExponent=3}
  n34 = Multiply(
   A=Multiply(A=Sign(None=@n25), B=OneMinus(None=@n33))
   B=Multiply(A=OneMinus(None=@n26), B=ScalarParameter() {defaultValue=0.1, parameterName=RefractionStrength})
  )
  Add(A=@n34, B=Constant2Vector() {r=0.01})
  n35 = TextureObjectParameter() {
   parameterName=Water_Mask
   samplerType=SAMPLERTYPE_LinearGrayscale
   texture=/FNGameplayCues/Athena/Artemis/Environments/BuildingActors/WaterFall/Materials/Texture/T_Plasma
  }
  n36 = TextureSample(UVs=@n22, Tex=@n35) {
   samplerType=SAMPLERTYPE_LinearGrayscale
   texture=&T_Plasma
  }
  n37 = Ceil(None=Subtract(A=OneMinus(None=@n36.R), B=@n31.R) {constB=0.9})
  n38 = SceneTexture(UVs=Add(A=@n1.ViewportUV, B=Multiply(A=@n34, B=@n37))) {sceneTextureId=PPI_PostProcessInput0}
  ShadingPathSwitch(Default=@n38.Color, Deferred=@n38.Color, Forward=@n38.Color, Mobile=SceneTexture() {sceneTextureId=PPI_PostProcessInput0}.Color)
 ]
}
